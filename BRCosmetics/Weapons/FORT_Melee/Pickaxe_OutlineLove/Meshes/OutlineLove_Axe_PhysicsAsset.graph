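# Physics asset for the OutlineLove pickaxe cosmetic's axe mesh, a two-handed melee weapon in the FORT_Melee set.
# It has a single capsule body on the Melee_TwoHanded_Handle bone, offset along the bone and rotated to lie along its length.
# It defines no constraints, so the weapon simulates and collides as one rigid piece.
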PhysicsAsset /BRCosmetics/Weapons/FORT_Melee/Pickaxe_OutlineLove/Meshes/OutlineLove_Axe_PhysicsAsset
bodies {Melee_TwoHanded_Handle [{shapeType=Capsule, length=54.1748, radius=9.8437, center=<-0.165 0.098 34.1598>, rotation=<rot 0.4616 -90 -0.0392>}]}
constraints []
